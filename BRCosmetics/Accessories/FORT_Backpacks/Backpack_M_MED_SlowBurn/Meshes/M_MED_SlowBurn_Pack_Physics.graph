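# Physics asset for the SlowBurn backpack cosmetic, giving collision bodies and joints to the pack's dyn_ bones for secondary motion.
# Capsules cover the backpack base, glasses, knot and two bow halves, while a sphere covers the tae pot.
# Swing- and twist-limited constraints join the glasses and pot to the base, and both bows to the knot.
# The knot and bow group has no constraint to the base.
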PhysicsAsset /BRCosmetics/Accessories/FORT_Backpacks/Backpack_M_MED_SlowBurn/Meshes/M_MED_SlowBurn_Pack_Physics
bodies {
 dyn_backpack_base [
  {shapeType=Capsule, length=17.5, radius=8.5, shapeName=dyn_backpack_base_capsule, center=<0 1.2435 0>, rotation=<rot -90 6.7214 -6.7215>}
  {
   shapeType=Capsule
   length=10
   radius=6
   shapeName=dyn_backpack_base_capsule
   center=<0 -0.0631 4.2238>
   rotation=<rot 90 -179.8417 -179.8421>
  }
 ]
 dyn_bow_l [{shapeType=Capsule, length=5.5, radius=2.5, shapeName=dyn_bow_l_capsule, center=<-5.6097 0 0>, rotation=<rot 90 179.9803 179.9801>}]
 dyn_bow_r [{shapeType=Capsule, length=4.5, radius=2, shapeName=dyn_bow_r_capsule, center=<-4.1685 0 0>, rotation=<rot 90 -179.9208 -179.9211>}]
 dyn_glasses [{shapeType=Capsule, length=7.5, radius=2.5, shapeName=dyn_glasses_capsule, center=<0 -1.0064 -4.413>, rotation=<rot 0 0 -10>}]
 dyn_knot [{shapeType=Capsule, length=0, radius=2.5, shapeName=dyn_knot_capsule}]
 dyn_tae_pot [{shapeType=Sphere, radius=7, shapeName=dyn_tae_pot_sphere, center=<-6.2282 -0.1512 0>}]
}
constraints [
 {bone1Name=dyn_glasses, bone2Name=dyn_backpack_base, swing1LimitDegrees=30, swing2LimitDegrees=30, twistLimitDegrees=10}
 {bone1Name=dyn_knot, bone2Name=dyn_bow_r, swing1LimitDegrees=15, swing2LimitDegrees=15, twistLimitDegrees=25}
 {bone1Name=dyn_knot, bone2Name=dyn_bow_l, swing1LimitDegrees=15, swing2LimitDegrees=15, twistLimitDegrees=25}
 {bone1Name=dyn_tae_pot, bone2Name=dyn_backpack_base, swing1LimitDegrees=30, swing2LimitDegrees=30, twistLimitDegrees=30}
]
