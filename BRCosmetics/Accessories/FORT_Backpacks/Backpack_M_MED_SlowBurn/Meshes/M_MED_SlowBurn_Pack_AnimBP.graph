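# Animation Blueprint for the SlowBurn backpack cosmetic, adding secondary physics motion and pose corrections to its dynamic bones.
# A base-bone offset feeds Anim Dynamics on the knot, then RigidBody simulation in base-bone space, then additive rotations on glasses and knot.
# The update event enables knot dynamics and additive rotations only in the front end.
# It also weakens RigidBody alpha during skydiving, and further with the parachute open.
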AnimBlueprint /BRCosmetics/Accessories/FORT_Backpacks/Backpack_M_MED_SlowBurn/Meshes/M_MED_SlowBurn_Pack_AnimBP
target_skeleton=&/M_MED_SlowBurn_Pack_Skeleton
graphs [
 AnimGraph {nodes [
   n1 = "Transform (Modify) Bone - Bone: dyn_backpack_base"(ComponentPose=Local To Component(LocalPose=Input Pose()), Alpha=1, Translation="-1.828,0.000,0.000", Rotation="0,0,0", Scale="1,1,1") {
    translationMode=BMM_Replace
    translationSpace=BCS_BoneSpace
    alphaBoolBlend {blendOption=Linear}
    boneToModify {boneName=dyn_backpack_base}
    translation {x=-1.828}
   }
   n2 = Anim Dynamics - Bone: dyn_knot(ComponentPose=@n1, Alpha=Get Dyn_alfa_AD() -> (Dyn_alfa_AD:double), ExternalForce="0,0,0") {
    angularSpringConstant=1500
    bAngularSpring=true
    bUseGravityOverride=true
    bUsePlanarLimit=false
    linearSpringConstant=10
    simulationSpace=World
    alphaBoolBlend {blendOption=Linear}
    alphaScaleBiasClamp {bInterpResult=true, interpSpeedDecreasing=1, interpSpeedIncreasing=1}
    boundBone {boneName=dyn_knot}
    componentAppliedLinearAccClamp=<1000 1000 1000>
    componentLinearAccScale=<1 1 1>
    componentLinearVelScale=<1 1 1>
    gravityOverride {z=-980}
    physicsBodyDefinitions [{
      boundBone {boneName=dyn_knot}
      boxExtents=<5 10 5>
      constraintSetup {twistAxis=AxisY, angularLimitsMax=<1 15 10>, angularLimitsMin=<-1 -5 -15>, angularTarget=<0 0.5 1.25>}
      localJointOffset=<0 0 0>
     }]
    relativeSpaceBone {boneName=dyn_backpack_base}
   }
   n3 = RigidBody(ComponentPose=@n2, Alpha=Get Dyn_alfa_RB() -> (Dyn_alfa_RB:double)) {
    bOverrideWorldGravity=true
    overridePhysicsAsset=&/M_MED_SlowBurn_Pack_Physics
    simulationSpace=BaseBoneSpace
    alphaBoolBlend {blendOption=Linear}
    baseBoneRef {boneName=dyn_backpack_base}
    componentAppliedLinearAccClamp=<1500 1500 1500>
    componentLinearAccScale=<0.75 0.75 0.75>
    componentLinearVelScale=<0.5 0.5 0.5>
    overrideWorldGravity {z=-980}
   }
   n4 = Get Dyn_alfa_frontend_tranform() -> (Dyn_alfa_frontend_tranform:double)
   n5 = "Transform (Modify) Bone - Bone: dyn_glasses"(ComponentPose=@n3, Alpha=@n4, Translation="0,0,0", Rotation="0,10,-5", Scale="1,1,1") {
    rotationMode=BMM_Additive
    rotationSpace=BCS_BoneSpace
    translationMode=BMM_Additive
    translationSpace=BCS_BoneSpace
    alphaBoolBlend {blendOption=Linear}
    alphaScaleBiasClamp {bInterpResult=true}
    boneToModify {boneName=dyn_glasses}
    rotation {roll=-5, yaw=10}
   }
   n6 = "Transform (Modify) Bone - Bone: dyn_knot"(ComponentPose=@n5, Alpha=@n4, Translation="0,0,0", Rotation="5,0,5", Scale="1,1,1") {
    rotationMode=BMM_Additive
    rotationSpace=BCS_BoneSpace
    translationMode=BMM_Additive
    translationSpace=BCS_BoneSpace
    alphaBoolBlend {blendOption=Linear}
    alphaScaleBiasClamp {bInterpResult=true}
    boneToModify {boneName=dyn_knot}
    rotation {pitch=5, roll=5}
   }
   Output Pose(Result=Component To Local(ComponentPose=@n6)) {layerGroup=None}
  ]}
 EventGraph {nodes [
   Event BlueprintUpdateAnimation()
   > n1 = Sequence()
   > n2 = Branch(Condition=Get bIsInFrontEnd())
   > Set Dyn_alfa_AD(Dyn_alfa_AD:double=1) -> (Output_Get:double)
   > Set Dyn_alfa_frontend_tranform(Dyn_alfa_frontend_tranform:double=1) -> (Output_Get:double)
   Set Dyn_alfa_AD(execute=@n2.else, Dyn_alfa_AD=0)
   > Set Dyn_alfa_frontend_tranform(Dyn_alfa_frontend_tranform=0)
   n3 = Branch(execute=@n1.then_1, Condition=Get bIsParachuteOpened())
   > Set Dyn_alfa_RB(Dyn_alfa_RB:double=0.25) -> (Output_Get:double)
   n4 = Branch(execute=@n3.else, Condition=Get bIsSkyDiving())
   > Set Dyn_alfa_RB(Dyn_alfa_RB=0.5)
   Set Dyn_alfa_RB(execute=@n4.else, Dyn_alfa_RB=1)
   TryGetPawnOwner()
  ]}
]
variables [
 Dyn_alfa_AD {type=number, default=1}
 Dyn_alfa_frontend_tranform {type=number, default=1}
 Dyn_alfa_RB {type=number, default=1}
]
